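# Simple looping MetaSound source for in-air projectile sounds, using projectile attenuation and the ranged weapon projectile sound class.
# On play, it picks a wave from an array with random pitch, scaled by output volume and affiliation gain.
# Default plays stereo from a random start time under an envelope fade-in; Mobile plays mono without either, at reduced gain.
# Random volume is computed but never reaches the audio.
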MetaSoundSource /FortMetaSoundWeapon/Items/Metasounds/Projectiles/MSS_Projectile_InAirLoop_Simple
inputs [
 Affiliation.Enemy {type=Float}
 Affiliation.Friendly {type=Float}
 Affiliation.Local {type=Float}
 Character.Affiliation {type=Enum:FortAudioAffiliation}
 Character.Health {type=Float}
 Character.UniqueID {type=Int32}
 Projectile.FadeInTime {type=Time}
 Projectile.OutputVolume {type=Float}
 Projectile.Sounds {type=WaveAsset:Array}
 "Random.Max Pitch (Semitones)" {type=Float}
 Random.MinPitch {type=Float}
 Random.VolumeMax {type=Float}
 Random.VolumeMin {type=Float}
 UE.Source.OnPlay {type=Trigger}
]
interfaces [Fort.Character:1.0, UE.OutputFormat.Stereo:1.0, UE.Source:1.1]
outputs [
 "UE.OutputFormat.Stereo.Audio:0" {type=Audio}
 "UE.OutputFormat.Stereo.Audio:1" {type=Audio}
]
pages [
 Default {
  is_build_graph=true
  input_defaults {
   Affiliation.Enemy=0
   Affiliation.Friendly=-6
   Affiliation.Local=0
   Character.Affiliation=0
   Character.Health=0
   Character.UniqueID=0
   Projectile.FadeInTime=0.3
   Projectile.OutputVolume=0
   "Random.Max Pitch (Semitones)"=0
   Random.MinPitch=0
   Random.VolumeMax=0
   Random.VolumeMin=0
   UE.Source.OnPlay=null
   Projectile.Sounds []
  }
  nodes [
   DecibelstoLinearGain_1 = UE.Decibels to Linear Gain:Float(Decibels=^Projectile.OutputVolume)
   ADSREnvelopeADSREnvelope_0 = ADSR Envelope.ADSR Envelope:Float(Attack Time=^Projectile.FadeInTime, Trigger Attack=^UE.Source.OnPlay) {also {defaults {Sustain Level=1}}}
   DecibelstoLinearGain_0 = UE.Decibels to Linear Gain:Float(Decibels=^Affiliation.Local)
   DecibelstoLinearGain_2 = UE.Decibels to Linear Gain:Float(Decibels=^Affiliation.Friendly)
   DecibelstoLinearGain_3 = UE.Decibels to Linear Gain:Float(Decibels=^Affiliation.Enemy)
   CrossfadeTriggerRouteFloat4_0 = "Crossfade.Trigger Route (Float, 4):Float"(
    Crossfade Value=Convert.Float:Int32(Int32=Convert.Int32:Enum:FortAudioAffiliation("Enum:FortAudioAffiliation"=^Character.Affiliation).Int32).Float
    In 0=@DecibelstoLinearGain_0.Linear Gain
    In 1=@DecibelstoLinearGain_2.Linear Gain
    In 2=@DecibelstoLinearGain_3.Linear Gain
    In 3=@DecibelstoLinearGain_3.Linear Gain
   )
   n1 = UE.Multiply:Float(
    PrimaryOperand=@DecibelstoLinearGain_1.Linear Gain
    AdditionalOperands=UE.Multiply:Float(AdditionalOperands=@ADSREnvelopeADSREnvelope_0.Out Envelope, PrimaryOperand=@CrossfadeTriggerRouteFloat4_0.Out).Out
   )
   n2 = UE.RandomFloat(Max=^"Random.Max Pitch (Semitones)", Min=^Random.MinPitch, Next=^UE.Source.OnPlay)
   n3 = UE.RandomFloat(Max=^Random.VolumeMax, Min=^Random.VolumeMin, Next=@n2.On Next)
   ArrayRandomGet_0 = Array.Random Get:WaveAsset:Array(In Array=^Projectile.Sounds, Next=@n3.On Next) {also {defaults {No Repeats=2}}}
   GetWaveDuration_0 = "UE.Get Wave Duration: "(Wave=@ArrayRandomGet_0.Value)
   n4 = UE.RandomTime(Next=@ArrayRandomGet_0.On Next, Max=@GetWaveDuration_0.Duration)
   n5 = InitVariable.Float()
   WavePlayer_0 = UE.Wave Player:Stereo(
    Wave Asset=@ArrayRandomGet_0.Value
    Play=@n4.On Next
    Start Time=@n4.Value
    Pitch Shift=VariableAccessor.Float(Variable=VariableMutator.Float(Variable=@n5.Variable, Value=@n2.Value).Variable).Value
   ) {also {defaults {Loop="<MetaSoundBuilderResult.SUCCEEDED: 0>"}}}
   Multiply_0 = UE.Multiply:Audio by Float(AdditionalOperands=@n1.Out, PrimaryOperand=@WavePlayer_0.Out Right)
   Multiply_1 = UE.Multiply:Audio by Float(AdditionalOperands=@n1.Out, PrimaryOperand=@WavePlayer_0.Out Left)
   DecibelstoLinearGain_4 = UE.Decibels to Linear Gain:Float(Decibels=@n3.Value)
   VariableMutator.Float(Value=@DecibelstoLinearGain_4.Linear Gain, Variable=InitVariable.Float().Variable)
  ]
  links [{dst="@UE.OutputFormat.Stereo.Audio:1", src=Multiply_0.Out}, {dst="@UE.OutputFormat.Stereo.Audio:0", src=Multiply_1.Out}]
 }
 Mobile {
  nodes [
   DecibelstoLinearGain_0 = UE.Decibels to Linear Gain:Float(Decibels=^Affiliation.Local)
   DecibelstoLinearGain_2 = UE.Decibels to Linear Gain:Float(Decibels=^Affiliation.Friendly)
   DecibelstoLinearGain_3 = UE.Decibels to Linear Gain:Float(Decibels=^Affiliation.Enemy)
   CrossfadeTriggerRouteFloat4_0 = "Crossfade.Trigger Route (Float, 4):Float"(
    Crossfade Value=Convert.Float:Int32(Int32=Convert.Int32:Enum:FortAudioAffiliation("Enum:FortAudioAffiliation"=^Character.Affiliation).Int32).Float
    In 0=@DecibelstoLinearGain_0.Linear Gain
    In 1=@DecibelstoLinearGain_2.Linear Gain
    In 2=@DecibelstoLinearGain_3.Linear Gain
    In 3=@DecibelstoLinearGain_3.Linear Gain
   )
   DecibelstoLinearGain_1 = UE.Decibels to Linear Gain:Float(Decibels=^Projectile.OutputVolume)
   n1 = UE.RandomFloat(Max=^"Random.Max Pitch (Semitones)", Min=^Random.MinPitch, Next=^UE.Source.OnPlay)
   n2 = UE.RandomFloat(Max=^Random.VolumeMax, Min=^Random.VolumeMin, Next=@n1.On Next)
   ArrayRandomGet_0 = Array.Random Get:WaveAsset:Array(In Array=^Projectile.Sounds, Next=@n2.On Next) {also {defaults {No Repeats=2}}}
   n3 = InitVariable.Float()
   WavePlayer_0 = UE.Wave Player:Mono(
    Play=@ArrayRandomGet_0.On Next
    Wave Asset=@ArrayRandomGet_0.Value
    Pitch Shift=VariableAccessor.Float(Variable=VariableMutator.Float(Variable=@n3.Variable, Value=@n1.Value).Variable).Value
   ) {also {defaults {Loop="<MetaSoundBuilderResult.SUCCEEDED: 0>"}}}
   Multiply_1 = UE.Multiply:Audio by Float(
    AdditionalOperands=UE.Multiply:Float(AdditionalOperands=@CrossfadeTriggerRouteFloat4_0.Out, PrimaryOperand=@DecibelstoLinearGain_1.Linear Gain).Out
    PrimaryOperand=@WavePlayer_0.Out Mono
   )
   Multiply_0 = UE.Multiply:Audio by Float(PrimaryOperand=@Multiply_1.Out) {also {defaults {AdditionalOperands=0.7}}}
   DecibelstoLinearGain_4 = UE.Decibels to Linear Gain:Float(Decibels=@n2.Value)
   VariableMutator.Float(Value=@DecibelstoLinearGain_4.Linear Gain, Variable=InitVariable.Float().Variable)
  ]
  links [{dst="@UE.OutputFormat.Stereo.Audio:0", src=Multiply_0.Out}, {dst="@UE.OutputFormat.Stereo.Audio:1", src=Multiply_0.Out}]
 }
]
properties {
 assetImportData=&self:AssetImportData
 attenuationSettings=/FortMetaSoundWeapon/Items/Shared/Settings/Projectiles/ATT_Projectile_InAir
 bLooping=true
 bOverrideConcurrency=true
 duration=10000
 maxDistance=12500
 numChannels=2
 outputFormat=Stereo
 sampleRate=48000
 soundClassObject=&Fort_WeaponRanged_Projectile_Effects
 busSends [{audioBus=&AB_Sidechain_FortWeapon_Vol, sendLevel=1}]
 concurrencyOverrides {maxCount=3}
 concurrencySet [null]
}
